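# Editor utility widget for testing spline mesh LOD generation on a chosen static mesh, defaulting to a helix test mesh.
# A Details View bound to the widget itself exposes the mesh and UV direction, above a "Run LOD Update" button.
# The button resolves U alignment via GetUVDirection for Auto or from the forced enum value, then calls GenerateSplineMeshLOD from LOD 0 and 1.
# A TestUVs node sits unconnected.
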EditorUtilityWidgetBlueprint /Game/NeverCook/TechArt/SplineMeshTools/Testing/EUW_TestingSplineMeshWorkflows
graphs [
 EventGraph {nodes [
   "On Pressed (btnLOD)"() {is=title K2Node_ComponentBoundEvent}
   > GenerateLODsOnTestMesh()
   Event Construct()
   > SetObject(self=Get detailsMeshToUpdate() -> (detailsMeshToUpdate:Details View Object Reference), NewObject=Self-Reference())
  ]}
 GenerateLODsOnTestMesh {nodes [
   GenerateLODsOnTestMesh() {is=title K2Node_FunctionEntry}
   n1 = Get StaticMeshAsset() -> (StaticMeshAsset:Static Mesh Object Reference)
   > Is Valid(InputObject=@n1)
   > n2 = Sequence()
   > n3 = Switch on SplineUVDirectionEnum!(execute, Selection:SplineUVDirectionEnum Enum=Get SplineUVDirection() -> (SplineUVDirection:SplineUVDirectionEnum Enum=NewEnumerator0)) -> (NewEnumerator0:Exec, NewEnumerator1:Exec, NewEnumerator2:Exec) {is=title K2Node_SwitchEnum}
   > n4 = GetUVDirection!(execute, StaticMeshAsset:Static Mesh Object Reference=@n1, __WorldContext:Object Reference) -> (then, AlignedToU:bool)
   > Set l_u(l_u:bool=@n4) -> (Output_Get:bool)
   Set l_u(execute=@n3.NewEnumerator1, l_u=true)
   Set l_u(execute=@n3.NewEnumerator2, l_u=false)
   n5 = Get l_u() -> (l_u:bool)
   GenerateSplineMeshLOD!(execute=@n2.then_1, FromLOD:int=0, StaticMeshAsset:Static Mesh Object Reference=@n1, AlignedToU:bool=@n5, __WorldContext:Object Reference) -> (then)
   > GenerateSplineMeshLOD(FromLOD=1, StaticMeshAsset=@n1, AlignedToU=@n5)
   TestUVs!(execute, FromLOD:int=0, StaticMeshAsset:Static Mesh Object Reference=@n1, __WorldContext:Object Reference) -> (then)
  ]}
]
variables [
 SplineUVDirection {type=SplineUVDirectionEnum, default=Auto}
 StaticMeshAsset {type=StaticMesh, default=/Game/NeverCook/TechArt/SplineMeshTools/Testing/Meshes/SM_Helix_ZAxis}
]
widgets {
 info {rootWidgetClass=&VerticalBox, widgetCount=4}
 widgets [
  VerticalBox_0 {widgetClassPath=&VerticalBox, widget=VerticalBox_0}
  detailsMeshToUpdate {
   widgetClassPath=&DetailsView
   parent=VerticalBox_0
   widget=detailsMeshToUpdate
   properties {propertiesToShow [StaticMeshAsset, SplineUVDirection]}
   slot {class=VerticalBoxSlot}
  }
  btnLOD {
   widgetClassPath=&Button
   parent=VerticalBox_0
   widget=btnLOD
   properties {widgetStyle {
     clickedSlateSound {resourceObject=null}
     disabled {
      bIsDynamicallyLoaded=false
      drawAs=RoundedBox
      imageType=NoImage
      mirroring=NoMirror
      resourceName=None
      resourceObject=null
      tiling=NoTile
      imageSize=<0 0>
      margin=<ltrb 0 0 0 0>
      outlineSettings {
       bUseBrushTransparency=false
       roundingType=FixedRadius
       width=1
       color {colorUseRule=UseColor_ColorTable, specifiedColor=<rgba 0 0 0 0>}
       cornerRadii=<4 4 4 4>
      }
      tintColor {colorUseRule=UseColor_Specified, specifiedColor=<rgba 0.0395 0.0395 0.0395 1>}
      uVRegion {bIsValid=false, max=<0 0>, min=<0 0>}
     }
     disabledForeground {colorUseRule=UseColor_ColorTable, specifiedColor=<rgba 0 0 0 0>}
     hovered {
      bIsDynamicallyLoaded=false
      drawAs=RoundedBox
      imageType=NoImage
      mirroring=NoMirror
      resourceName=None
      resourceObject=null
      tiling=NoTile
      imageSize=<32 32>
      margin=<ltrb 0 0 0 0>
      outlineSettings {
       bUseBrushTransparency=false
       roundingType=FixedRadius
       width=1
       color {colorUseRule=UseColor_ColorTable, specifiedColor=<rgba 0 0 0 0>}
       cornerRadii=<4 4 4 4>
      }
      tintColor {colorUseRule=UseColor_Specified, specifiedColor=<rgba 0.0953 0.0953 0.0953 1>}
      uVRegion {bIsValid=false, max=<0 0>, min=<0 0>}
     }
     hoveredForeground {colorUseRule=UseColor_ColorTable, specifiedColor=<rgba 0 0 0 0>}
     hoveredSlateSound {resourceObject=null}
     normal {
      bIsDynamicallyLoaded=false
      drawAs=RoundedBox
      imageType=NoImage
      mirroring=NoMirror
      resourceName=None
      resourceObject=null
      tiling=NoTile
      imageSize=<32 32>
      margin=<ltrb 0 0 0 0>
      outlineSettings {
       bUseBrushTransparency=false
       roundingType=FixedRadius
       width=1
       color {colorUseRule=UseColor_ColorTable, specifiedColor=<rgba 0 0 0 0>}
       cornerRadii=<4 4 4 4>
      }
      tintColor {colorUseRule=UseColor_Specified, specifiedColor=<rgba 0.0395 0.0395 0.0395 1>}
      uVRegion {bIsValid=false, max=<0 0>, min=<0 0>}
     }
     normalForeground {colorUseRule=UseColor_ColorTable, specifiedColor=<rgba 0 0 0 0>}
     normalPadding=<ltrb 12 1.5 12 1.5>
     pressed {
      bIsDynamicallyLoaded=false
      drawAs=RoundedBox
      imageType=NoImage
      mirroring=NoMirror
      resourceName=None
      resourceObject=null
      tiling=NoTile
      imageSize=<32 32>
      margin=<ltrb 0 0 0 0>
      outlineSettings {
       bUseBrushTransparency=false
       roundingType=FixedRadius
       width=1
       color {colorUseRule=UseColor_ColorTable, specifiedColor=<rgba 0 0 0 0>}
       cornerRadii=<4 4 4 4>
      }
      tintColor {colorUseRule=UseColor_Specified, specifiedColor=<rgba 0.0284 0.0284 0.0284 1>}
      uVRegion {bIsValid=false, max=<0 0>, min=<0 0>}
     }
     pressedForeground {colorUseRule=UseColor_ColorTable, specifiedColor=<rgba 0 0 0 0>}
     pressedPadding=<ltrb 12 2.5 12 0.5>
     pressedSlateSound {resourceObject=null}
    }}
   slot {class=VerticalBoxSlot}
  }
  TextBlock_56 {
   widgetClassPath=&TextBlock
   parent=btnLOD
   widget=TextBlock_56
   properties {
    text=Run LOD Update
    colorAndOpacity {colorUseRule=UseColor_Specified, specifiedColor=<rgba 0.8594 0.8594 0.8594 1>}
    font {
     bForceMonospaced=false
     bMaterialIsStencil=false
     fontMaterial=null
     fontObject=&Roboto
     letterSpacing=0
     monospacedWidth=1
     size=13.5
     skewAmount=0
     typefaceFontName=Bold
     outlineSettings {
      bApplyOutlineToDropShadows=false
      bMiteredCorners=false
      bSeparateFillAlpha=false
      outlineMaterial=null
      outlineSize=0
      outlineColor=<rgba 0 0 0 1>
     }
    }
   }
   slot {class=ButtonSlot}
  }
 ]
}
