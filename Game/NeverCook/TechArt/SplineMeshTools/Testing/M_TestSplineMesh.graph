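# Unlit, two-sided wireframe test material in the SplineMeshTools testing folder, with no connections to material outputs shown.
# A vertex interpolator carries world position offset along the normal by Base times the V coordinate remapped from 0..1 to -1..1.
# Its distance from the pixel's world position, divided by Dist, feeds separate Lerp and Saturate nodes.
# A VertexColor node sits unconnected.
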Material /Game/NeverCook/TechArt/SplineMeshTools/Testing/M_TestSplineMesh
graph {
 properties {
  bAutomaticallySetUsageInEditor=false
  bCastRayTracedShadows=false
  shadingModel=MSM_Unlit
  twoSided=true
  usedShadingModels=Unlit
  wireframe=true
 }
 nodes [
  n1 = VertexNormalWS()
  n2 = TextureCoordinate()
  n3 = Multiply(A=ScalarParameter() {defaultValue=0.1, parameterName=Base}, B=ConstantBiasScale(None=ComponentMask(None=@n2) {g=true}) {bias=-0.5, scale=2})
  n4 = VertexInterpolator(VS=Add(A=WorldPosition().XYZ, B=Multiply(A=@n1, B=@n3)))
  n5 = Divide(
   A=Distance(A=ComponentMask(None=@n4.PS) {b=true, g=true, r=true}, B=WorldPosition().XYZ)
   B=ScalarParameter() {defaultValue=0.001, parameterName=Dist}
  )
  LinearInterpolate(Alpha=@n5) {constA=1, constB=0.05}
  Saturate(None=@n5)
  VertexColor()
 ]
}
